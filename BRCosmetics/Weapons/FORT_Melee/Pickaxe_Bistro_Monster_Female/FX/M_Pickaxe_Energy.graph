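# Translucent, unlit, two-sided energy material in the Bistro Monster Female pickaxe FX, enabled for Niagara ribbons and mesh particles.
# Panned smoke tints a dark-to-pink emissive, while a noise-distorted trail mask, UV edge fades, particle alpha and depth fade build opacity.
# Dynamic parameters drive pan speed, tiling and time offset; a sine wave scaled by WPO adds wobble.
# Glider_Trail blends in a Fresnel tint, layered smoke and noise, and a different fade shape.
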Material /BRCosmetics/Weapons/FORT_Melee/Pickaxe_Bistro_Monster_Female/FX/M_Pickaxe_Energy
graph {
 properties {
  bUsedWithNiagaraMeshParticles=true
  bUsedWithNiagaraRibbons=true
  blendMode=BLEND_Translucent
  pixelDepthOffsetMode=PDOM_Legacy
  shadingModel=MSM_Unlit
  twoSided=true
  usedShadingModels=Unlit
 }
 nodes [
  Constant()
  n1 = Time()
  n2 = DynamicParameter() {defaultValue=<rgba 1 1 2 1>}
  n3 = Multiply(A=Constant() {r=1}, B=DynamicParameter() {defaultValue=<rgba 1 1 2 1>}.Param1)
  n4 = Constant()
  n5 = TextureSample(UVs=Panner(Time=Add(A=@n1, B=@n2.Param3), Speed=AppendVector(A=@n3, B=@n4)) {speedX=1}) {texture=&T_Inky_Smoke_Tile}
  n6 = LinearInterpolate(A=Constant3Vector() {constant=<rgba 0.0156 0 0 0>}, B=Constant3Vector() {constant=<rgba 1 0 0.2573 0>}, Alpha=@n5.R)
  n7 = Constant3Vector() {constant=<rgba 0.3958 0.0033 0.0627 0>}
  n8 = Fresnel() {baseReflectFraction=0}
  MaterialFunctionCall(Emissive=LinearInterpolate(A=@n6, B=LinearInterpolate(A=@n6, B=@n7, Alpha=@n8) {constA=1}, Alpha=ScalarParameter() {parameterName=Glider_Trail})) {materialFunction=&MF_FX_EmissiveAdjustments}
  Multiply(A=@n5.R, B=ComponentMask(None=TextureCoordinate()) {r=true})
  n9 = ComponentMask(None=TextureCoordinate()) {r=true}
  n10 = Time()
  n11 = Sine(None=Multiply(A=Add(A=@n9, B=@n10)) {constB=1.4})
  Multiply(
   A=MaterialFunctionCall(X=@n11, Y=@n11, Z=@n11) {materialFunction=&MakeFloat3}.Result
   B=ScalarParameter() {defaultValue=1.5, parameterName=WPO}
  ) {constB=1.5}
  n12 = DynamicParameter() {defaultValue=<rgba 1 1 2 1>}
  n13 = Constant() {r=1}
  n14 = Multiply(A=TextureCoordinate(), B=AppendVector(A=Multiply(A=@n12.Param2, B=@n13), B=Constant() {r=1}))
  n15 = TextureCoordinate() {uTiling=0.6, vTiling=0.5}
  n16 = Multiply(A=TextureSample(UVs=Panner(Coordinate=@n15) {speedX=0.5}) {samplerType=SAMPLERTYPE_Normal, texture=/Game/Global/Textures/Default/Noise/T_SwirlNoise_N}.R) {constB=0.2}
  n17 = DynamicParameter() {defaultValue=<rgba 1 1 2 1>}
  n18 = Add(A=Time(), B=@n17.Param3)
  n19 = AppendVector(A=Multiply(A=Constant() {r=2}, B=@n17.Param1), B=Constant())
  n20 = TextureSample(UVs=Panner(Coordinate=Add(A=@n14, B=@n16), Time=@n18, Speed=@n19) {speedX=2}) {texture=/Game/Effects/Fort_Effects/Textures/Smoke/T_Lobber_Trail_Mask}
  n21 = TextureSample(UVs=Panner(Coordinate=TextureCoordinate() {uTiling=0.2, vTiling=0.35}, Time=@n18, Speed=@n19)) {texture=/Game/Effects/Fort_Effects/Textures/Tile/T_Inky_Smoke_Tile2}
  n22 = TextureSample(UVs=Panner(Coordinate=@n14, Time=@n18) {speedX=0.9}) {samplerType=SAMPLERTYPE_LinearGrayscale, texture=/Engine/Functions/Engine_MaterialFunctions02/ExampleContent/Textures/LowResBlurredNoise}
  n23 = TextureCoordinate()
  n24 = Multiply(A=@n22.RGB, B=OneMinus(None=ComponentMask(None=@n23) {r=true}))
  n25 = ScalarParameter() {parameterName=Glider_Trail}
  n26 = LinearInterpolate(
   A=@n20.R
   B=Add(
    A=MaterialFunctionCall!(Blend=@n20.RGB, Base=@n21.RGB) {materialFunction=/Engine/Functions/Engine_MaterialFunctions03/Blends/Blend_Screen}.Result
    B=Multiply(A=@n24) {constB=2}
   )
   Alpha=@n25
  )
  n27 = TextureCoordinate()
  n28 = ComponentMask(None=@n27) {r=true}
  n29 = Constant()
  n30 = Constant() {r=1}
  n31 = Multiply(A=ComponentMask(None=@n27) {g=true}) {constB=0.5}
  n32 = LinearInterpolate(
   A=Sine(None=Multiply(A=@n28) {constB=0.5})
   B=Multiply(A=SphereMask(A=@n28, B=@n29, Radius=@n30) {attenuationRadius=1, hardnessPercent=0}, B=Sine(None=@n31))
   Alpha=@n25
  )
  Multiply(A=ParticleColor().A, B=DepthFade(Opacity=Multiply(A=@n26, B=@n32), FadeDistance=ScalarParameter() {defaultValue=36}))
 ]
}
